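# M_DLR_Lotusflow is an unlit masked material for the LotusFlow dealership set's glowing lotus petals, enabled for skeletal meshes and Niagara mesh particles.
# Packed textures supply petal outline, fill and inner glow, combined with caustics distorted by panning Perlin noise for an adjustable emissive.
# An IsParticle parameter selects particle color or painted-variant color.
# Mask math adds outline opacity, a dynamic-parameter fade and noise-based erosion.
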Material /DealershipSharedLibrary/Assets/Sets/LotusFlow/Materials/M_DLR_Lotusflow
graph {
 properties {
  bAutomaticallySetUsageInEditor=false
  bUsedWithMeshDeformer=true
  bUsedWithNiagaraMeshParticles=true
  bUsedWithSkeletalMesh=true
  blendMode=BLEND_Masked
  shadingModel=MSM_Unlit
  usedShadingModels=Unlit
 }
 nodes [
  n1 = NamedRerouteUsage()
  n2 = ScalarParameter() {defaultValue=150, parameterName=OutlineEmissionMult}
  n3 = NamedRerouteUsage()
  n4 = TextureSample() {texture=/DealershipSharedLibrary/Assets/Sets/LotusFlow/Textures/T_DLR_Lotusflow_RGB}
  n5 = TextureSample() {texture=/DealershipSharedLibrary/Assets/Sets/LotusFlow/Textures/T_DLR_Lotusflow_CenterGlow_Details}
  n6 = Power(Base=@n5.R, Exp=ScalarParameter() {defaultValue=1.7, parameterName=InnerGlowPow})
  n7 = ScalarParameter() {defaultValue=4, parameterName=InnerGlowMult}
  n8 = TextureCoordinate()
  n9 = Time()
  n10 = ScalarParameter() {defaultValue=0.5, parameterName=PanSpeed}
  n11 = TextureSample(UVs=Panner(Coordinate=@n8, Time=Multiply(A=@n9, B=@n10)) {speedX=0.2}) {samplerType=SAMPLERTYPE_LinearGrayscale, texture=&T_SimplePerlin_Greyscale}
  n12 = ScalarParameter() {defaultValue=0.3119, parameterName=Distortion}
  n13 = TextureSample(UVs=Add(A=Multiply(A=@n11.G, B=@n12), B=@n8)) {texture=&T_ShieldCaustics}
  n14 = ScalarParameter() {defaultValue=1.7588, parameterName=CausticNoisePower}
  n15 = Add(
   A=Multiply(A=@n5.G, B=ScalarParameter() {defaultValue=0.98, parameterName=TransluscentGlowMult})
   B=Multiply(A=Power(Base=@n13.G, Exp=@n14), B=ScalarParameter() {defaultValue=1, parameterName=EmissionNoiseMult})
  )
  n16 = NamedRerouteDeclaration(None=@n4.A) {name=TextureSampleAlpha, nodeColor=<rgba 0.8419 0.9086 1 1>}
  n17 = Multiply(
   A=Power(Base=Saturate(None=@n16), Exp=ScalarParameter() {defaultValue=1, parameterName=TransluscentGlowPow})
   B=ScalarParameter() {parameterName=TransluscentGlowInfluence}
  )
  n18 = Add(
   A=LinearInterpolate(
    A=Desaturation(None=@n4.RGBA) {luminanceFactors=<rgba 0.3 0.6 0.1 1>}
    B=Multiply(A=@n6, B=@n7)
    Alpha=ScalarParameter() {defaultValue=1.1124, parameterName=GlowBlend}
   )
   B=Power(Base=Add(A=@n15, B=@n17), Exp=ScalarParameter() {defaultValue=0.94, parameterName=EmissionPower})
  )
  n19 = Add(
   A=Multiply(A=Multiply(A=@n1, B=@n2), B=@n3)
   B=Multiply(A=Multiply(A=@n3, B=@n18), B=ScalarParameter() {defaultValue=1.7, parameterName=GlowOverTime})
  )
  n20 = ScalarParameter() {parameterName=Desaturation}
  MaterialFunctionCall(Emissive=Multiply(
    A=Desaturation(None=@n19, Fraction=@n20) {luminanceFactors=<rgba 0.3 0.6 0.1 1>}
    B=ScalarParameter() {defaultValue=1, parameterName=OverallBrightness}
   )) {materialFunction=&MF_FX_EmissiveAdjustments}
  n21 = NamedRerouteUsage()
  n22 = Power(Base=@n21, Exp=ScalarParameter() {defaultValue=2, parameterName=OpacityMaskPower})
  n23 = Multiply(A=NamedRerouteUsage(), B=@n22)
  n24 = ScalarParameter() {parameterName=OutlineOpacity}
  Multiply(A=Add(A=@n22, B=Multiply(A=@n23, B=@n24)), B=DynamicParameter() {paramNames [OpacityFade,Param2,Param3,Param4]}.OpacityFade)
  NamedRerouteDeclaration(None=@n4.R) {name=PetalsOutline, nodeColor=<rgba 1 0 0.3239 1>}
  NamedRerouteDeclaration(None=@n4.B) {name=PetalsFill, nodeColor=<rgba 0 0.4092 1 1>}
  NamedRerouteDeclaration(None=@n5.R) {name=InnerGlow, nodeColor=<rgba 0 1 0.391 1>}
  n25 = VectorParameter() {parameterName=Color, defaultValue=<rgba 2.1865 0.428 1.1128 1>}
  n26 = Multiply(A=MaterialFunctionCall!(Primary Color=AppendVector(A=@n25.RGB, B=@n25.RGB), Tertiary Color, "Secondary  Color") {materialFunction=/DealershipSharedLibrary/FX/Materials/MaterialFunctions/MF_DLR_PaintedVariantsData}.Primary RGB)
  NamedRerouteDeclaration(None=If(A=ScalarParameter() {parameterName=IsParticle}, "A > B"=ParticleColor().RGB, "A == B"=@n26, "A < B"=@n26)) {name=finalcolor, nodeColor=<rgba 0.5234 0 1 1>}
  OneMinus(None=@n21)
  n27 = Add(A=NamedRerouteUsage(), B=Add(A=NamedRerouteUsage(), B=NamedRerouteUsage()))
  n28 = Add(A=@n11.R, B=ScalarParameter() {defaultValue=-1, parameterName=LotusErode})
  Reroute(None=Subtract(A=Saturate(None=@n27), B=Saturate(None=@n28)))
 ]
}
