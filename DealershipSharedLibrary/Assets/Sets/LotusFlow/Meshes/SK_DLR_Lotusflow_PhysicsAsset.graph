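# Physics asset for the Lotus Flow set-piece skeleton in the dealership shared library, giving the lotus capsule bodies and joints.
# Twelve three-body chains, four each in the Back, Mid and Front groups, branch from a tiny lotus_base body that serves as the hub.
# Each segment is one large capsule constrained to its parent, with identical 45-degree swing and twist limit values on every joint.
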PhysicsAsset /DealershipSharedLibrary/Assets/Sets/LotusFlow/Meshes/SK_DLR_Lotusflow_PhysicsAsset
bodies {
 Lotus_Back_JT_01 [{shapeType=Capsule, length=146.0121, radius=143.1418, center=<98.0195 -14.9294 -17.5348>, rotation=<rot -2.5552 -90 -33.2121>}]
 Lotus_Back_JT_02 [{shapeType=Capsule, length=190.8613, radius=99.2276, center=<127.6969 -13.4456 0.6797>, rotation=<rot -0.1263 -90 -0.174>}]
 Lotus_Back_JT_03 [{shapeType=Capsule, length=125.9111, radius=84.8598, center=<78.8665 -8.5267 0.5266>, rotation=<rot -1.0081 -90 -0.6092>}]
 Lotus_Back_JT_05 [{shapeType=Capsule, length=146.2401, radius=145.0348, center=<94.7815 -12.6077 20.0542>, rotation=<rot 0 -85.5418 -139.4319>}]
 Lotus_Back_JT_06 [{shapeType=Capsule, length=190.8603, radius=99.2277, center=<120.9996 -14.3177 2.873>, rotation=<rot -0.0228 -90 0.7609>}]
 Lotus_Back_JT_07 [{shapeType=Capsule, length=116.3803, radius=66.6943, center=<90.1583 -9.688 -1.1822>, rotation=<rot -0.8478 -90 0.6951>}]
 Lotus_Back_JT_09 [{shapeType=Capsule, length=146.4047, radius=144.0105, center=<112.5053 -15.7019 -16.0191>, rotation=<rot -3.0825 -90 -37.4874>}]
 Lotus_Back_JT_10 [{shapeType=Capsule, length=190.7284, radius=91.1123, center=<130.7135 -12.2013 -2.9014>, rotation=<rot -0.3383 -90 1.3208>}]
 Lotus_Back_JT_11 [{shapeType=Capsule, length=141.9944, radius=109.1639, center=<52.8153 -9.5751 -20.8666>, rotation=<rot 1.0035 -90 19.8413>}]
 Lotus_Back_JT_13 [{shapeType=Capsule, length=140.9614, radius=134.468, center=<98.2232 -13.9682 17.9544>, rotation=<rot 0 -85.1059 -59.7051>}]
 Lotus_Back_JT_14 [{shapeType=Capsule, length=190.5825, radius=96.0088, center=<104.9374 -12.2627 1.2029>, rotation=<rot -0.4226 -90 -3.1717>}]
 Lotus_Back_JT_15 [{shapeType=Capsule, length=151.9023, radius=87.7989, center=<89.2026 -9.2897 0.3567>, rotation=<rot -0.9163 -90 -0.6325>}]
 Lotus_Front_JT_01 [{shapeType=Capsule, length=163.4426, radius=127.0271, center=<60.6667 -13.1604 -5.0266>, rotation=<rot 0 79.8621 -172.2313>}]
 Lotus_Front_JT_02 [{shapeType=Capsule, length=190.9204, radius=83.0128, center=<88.3887 -26.3427 6.5462>, rotation=<rot 0.5624 -90 0.8349>}]
 Lotus_Front_JT_03 [{shapeType=Capsule, length=151.9033, radius=82.7319, center=<87.253 -3.5157 3.8683>, rotation=<rot 0.0985 -90 0.6197>}]
 Lotus_Front_JT_06 [{shapeType=Capsule, length=190.9058, radius=75.2971, center=<105.6784 -15.9435 0.5671>, rotation=<rot 1.0416 -90 1.5247>}]
 Lotus_Front_JT_07 [{shapeType=Capsule, length=122.7453, radius=71.937, center=<73.0086 0.8956 8.0176>, rotation=<rot -0.2912 -90 -5.6407>}]
 Lotus_Front_JT_09 [{shapeType=Capsule, length=167.6575, radius=157.6511, center=<88.2593 -17.0946 15.3062>, rotation=<rot 0 -98.54 -170.2675>}]
 Lotus_Front_JT_10 [{shapeType=Capsule, length=190.8946, radius=102.8565, center=<102.7019 -27.3997 5.6316>, rotation=<rot -0.0553 -90 -0.9644>}]
 Lotus_Front_JT_11 [{shapeType=Capsule, length=151.891, radius=83.5418, center=<80.6483 -1.7841 6.0622>, rotation=<rot -0.0168 -90 -0.7662>}]
 Lotus_Front_JT_13 [{shapeType=Capsule, length=164.0015, radius=125.8575, center=<60.0244 -13.8365 21.3098>, rotation=<rot 0 -99.1379 -170.1084>}]
 Lotus_Front_JT_14 [{shapeType=Capsule, length=190.9204, radius=82.9922, center=<95.6865 -22.8698 1.3928>, rotation=<rot -0.2225 -90 -0.6969>}]
 Lotus_Front_JT_15 [{shapeType=Capsule, length=151.9033, radius=76.4808, center=<94.5395 13.0549 4.4352>, rotation=<rot -0.2053 -90 -1.0795>}]
 Lotus_Front_JT_17 [{shapeType=Capsule, length=182.0541, radius=148.5219, center=<79.665 -14.0853 7.3245>, rotation=<rot 0 -93.7457 -179.0349>}]
 Lotus_Mid_JT_01 [{shapeType=Capsule, length=164.3114, radius=124.1725, center=<40.5548 -3.9305 -11.2884>, rotation=<rot 0 86.8459 -173.3724>}]
 Lotus_Mid_JT_02 [{shapeType=Capsule, length=190.9216, radius=80.1641, center=<88.8827 -14.8491 1.7819>, rotation=<rot 0.1085 -90 -0.4133>}]
 Lotus_Mid_JT_03 [{shapeType=Capsule, length=174.8832, radius=99.9509, center=<89.8506 4.7792 3.8047>, rotation=<rot -0.0025 -90 -0.8114>}]
 Lotus_Mid_JT_06 [{shapeType=Capsule, length=190.9215, radius=80.161, center=<89.3115 -21.1025 2.8837>, rotation=<rot 0.0934 -90 -0.4641>}]
 Lotus_Mid_JT_07 [{shapeType=Capsule, length=174.8832, radius=99.9545, center=<89.1583 -1.4678 5.0654>, rotation=<rot -0.0083 -90 -0.864>}]
 Lotus_Mid_JT_10 [{shapeType=Capsule, length=190.9215, radius=80.165, center=<88.8879 -14.8462 -0.8103>, rotation=<rot 0.1473 -90 0.1622>}]
 Lotus_Mid_JT_11 [{shapeType=Capsule, length=174.8832, radius=99.9508, center=<89.8678 4.7802 -0.5072>, rotation=<rot -0.0668 -90 -0.2381>}]
 Lotus_Mid_JT_13 [{shapeType=Capsule, length=149.8803, radius=145.215, center=<85.2882 -7.3269 27.4095>, rotation=<rot 0 -94.0246 -152.0138>}]
 Lotus_Mid_JT_14 [{shapeType=Capsule, length=190.9215, radius=80.1606, center=<89.0421 -17.1258 1.1425>, rotation=<rot 0.1181 -90 -0.2713>}]
 Lotus_Mid_JT_15 [{shapeType=Capsule, length=174.8832, radius=99.9551, center=<89.6073 2.4897 2.7409>, rotation=<rot -0.0183 -90 -0.6699>}]
 lotus_Mid_JT_05 [{shapeType=Capsule, length=164.0587, radius=138.3291, center=<28.8666 -4.5324 -13.365>, rotation=<rot 0 85.1225 -172.5363>}]
 lotus_Mid_leaf_09 [{shapeType=Capsule, length=183.1632, radius=149.1888, center=<62.0118 -4.7333 41.3496>, rotation=<rot 0 -93.6936 -57.4342>}]
 lotus_base [{shapeType=Capsule, length=0.505, radius=0.505, rotation=<rot 0 -90 0>}]
}
constraints [
 {bone1Name=Lotus_Back_JT_01, bone2Name=lotus_base, swing1LimitDegrees=45, swing2LimitDegrees=45, twistLimitDegrees=45}
 {bone1Name=Lotus_Back_JT_02, bone2Name=Lotus_Back_JT_01, swing1LimitDegrees=45, swing2LimitDegrees=45, twistLimitDegrees=45}
 {bone1Name=Lotus_Back_JT_03, bone2Name=Lotus_Back_JT_02, swing1LimitDegrees=45, swing2LimitDegrees=45, twistLimitDegrees=45}
 {bone1Name=Lotus_Back_JT_05, bone2Name=lotus_base, swing1LimitDegrees=45, swing2LimitDegrees=45, twistLimitDegrees=45}
 {bone1Name=Lotus_Back_JT_06, bone2Name=Lotus_Back_JT_05, swing1LimitDegrees=45, swing2LimitDegrees=45, twistLimitDegrees=45}
 {bone1Name=Lotus_Back_JT_07, bone2Name=Lotus_Back_JT_06, swing1LimitDegrees=45, swing2LimitDegrees=45, twistLimitDegrees=45}
 {bone1Name=Lotus_Back_JT_09, bone2Name=lotus_base, swing1LimitDegrees=45, swing2LimitDegrees=45, twistLimitDegrees=45}
 {bone1Name=Lotus_Back_JT_10, bone2Name=Lotus_Back_JT_09, swing1LimitDegrees=45, swing2LimitDegrees=45, twistLimitDegrees=45}
 {bone1Name=Lotus_Back_JT_11, bone2Name=Lotus_Back_JT_10, swing1LimitDegrees=45, swing2LimitDegrees=45, twistLimitDegrees=45}
 {bone1Name=Lotus_Back_JT_13, bone2Name=lotus_base, swing1LimitDegrees=45, swing2LimitDegrees=45, twistLimitDegrees=45}
 {bone1Name=Lotus_Back_JT_14, bone2Name=Lotus_Back_JT_13, swing1LimitDegrees=45, swing2LimitDegrees=45, twistLimitDegrees=45}
 {bone1Name=Lotus_Back_JT_15, bone2Name=Lotus_Back_JT_14, swing1LimitDegrees=45, swing2LimitDegrees=45, twistLimitDegrees=45}
 {bone1Name=Lotus_Mid_JT_01, bone2Name=lotus_base, swing1LimitDegrees=45, swing2LimitDegrees=45, twistLimitDegrees=45}
 {bone1Name=Lotus_Mid_JT_02, bone2Name=Lotus_Mid_JT_01, swing1LimitDegrees=45, swing2LimitDegrees=45, twistLimitDegrees=45}
 {bone1Name=Lotus_Mid_JT_03, bone2Name=Lotus_Mid_JT_02, swing1LimitDegrees=45, swing2LimitDegrees=45, twistLimitDegrees=45}
 {bone1Name=lotus_Mid_JT_05, bone2Name=lotus_base, swing1LimitDegrees=45, swing2LimitDegrees=45, twistLimitDegrees=45}
 {bone1Name=Lotus_Mid_JT_06, bone2Name=lotus_Mid_JT_05, swing1LimitDegrees=45, swing2LimitDegrees=45, twistLimitDegrees=45}
 {bone1Name=Lotus_Mid_JT_07, bone2Name=Lotus_Mid_JT_06, swing1LimitDegrees=45, swing2LimitDegrees=45, twistLimitDegrees=45}
 {bone1Name=lotus_Mid_leaf_09, bone2Name=lotus_base, swing1LimitDegrees=45, swing2LimitDegrees=45, twistLimitDegrees=45}
 {bone1Name=Lotus_Mid_JT_10, bone2Name=lotus_Mid_leaf_09, swing1LimitDegrees=45, swing2LimitDegrees=45, twistLimitDegrees=45}
 {bone1Name=Lotus_Mid_JT_11, bone2Name=Lotus_Mid_JT_10, swing1LimitDegrees=45, swing2LimitDegrees=45, twistLimitDegrees=45}
 {bone1Name=Lotus_Mid_JT_13, bone2Name=lotus_base, swing1LimitDegrees=45, swing2LimitDegrees=45, twistLimitDegrees=45}
 {bone1Name=Lotus_Mid_JT_14, bone2Name=Lotus_Mid_JT_13, swing1LimitDegrees=45, swing2LimitDegrees=45, twistLimitDegrees=45}
 {bone1Name=Lotus_Mid_JT_15, bone2Name=Lotus_Mid_JT_14, swing1LimitDegrees=45, swing2LimitDegrees=45, twistLimitDegrees=45}
 {bone1Name=Lotus_Front_JT_01, bone2Name=lotus_base, swing1LimitDegrees=45, swing2LimitDegrees=45, twistLimitDegrees=45}
 {bone1Name=Lotus_Front_JT_02, bone2Name=Lotus_Front_JT_01, swing1LimitDegrees=45, swing2LimitDegrees=45, twistLimitDegrees=45}
 {bone1Name=Lotus_Front_JT_03, bone2Name=Lotus_Front_JT_02, swing1LimitDegrees=45, swing2LimitDegrees=45, twistLimitDegrees=45}
 {bone1Name=Lotus_Front_JT_17, bone2Name=lotus_base, swing1LimitDegrees=45, swing2LimitDegrees=45, twistLimitDegrees=45}
 {bone1Name=Lotus_Front_JT_06, bone2Name=Lotus_Front_JT_17, swing1LimitDegrees=45, swing2LimitDegrees=45, twistLimitDegrees=45}
 {bone1Name=Lotus_Front_JT_07, bone2Name=Lotus_Front_JT_06, swing1LimitDegrees=45, swing2LimitDegrees=45, twistLimitDegrees=45}
 {bone1Name=Lotus_Front_JT_09, bone2Name=lotus_base, swing1LimitDegrees=45, swing2LimitDegrees=45, twistLimitDegrees=45}
 {bone1Name=Lotus_Front_JT_10, bone2Name=Lotus_Front_JT_09, swing1LimitDegrees=45, swing2LimitDegrees=45, twistLimitDegrees=45}
 {bone1Name=Lotus_Front_JT_11, bone2Name=Lotus_Front_JT_10, swing1LimitDegrees=45, swing2LimitDegrees=45, twistLimitDegrees=45}
 {bone1Name=Lotus_Front_JT_13, bone2Name=lotus_base, swing1LimitDegrees=45, swing2LimitDegrees=45, twistLimitDegrees=45}
 {bone1Name=Lotus_Front_JT_14, bone2Name=Lotus_Front_JT_13, swing1LimitDegrees=45, swing2LimitDegrees=45, twistLimitDegrees=45}
 {bone1Name=Lotus_Front_JT_15, bone2Name=Lotus_Front_JT_14, swing1LimitDegrees=45, swing2LimitDegrees=45, twistLimitDegrees=45}
]
